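# Niagara system for the WealthLamb exit effect, combining a burst of stretching meshes with a large ground smoke ring.
# DirectionalBurst repeatedly spawns short-lived meshes below the origin that scale up and fade.
# Smoke_Sphere launches sprites outward from a ring at System.GroundPosition, with drag, upward drift, slowing spin and curve-driven dynamic material parameters.
# Ripple, distortion, post-process and mesh airflow emitters are disabled; the enabled SoftAirflow emitter has no renderers.
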NiagaraSystem /PistolCanary/Assets/FX/WealthLamb/Exit/NS_PC_WealthLamb_Exit
dependencies {
 usedDataInterfaces [&NiagaraDataInterfaceCurve, &NiagaraDataInterfaceVectorCurve]
 usedDynamicInputs [
  &SubtractVectorFromPosition
  /Niagara/DynamicInputs/Helpers/SimulationPosition
  &Multiply_Vector
 ]
 usedModules [
  &self:SystemSpawnScript.SystemScriptSource.SystemScriptGraph.NiagaraNodeAssignment_0.SetVariables_#1
  &self:SystemSpawnScript.SystemScriptSource.SystemScriptGraph.NiagaraNodeAssignment_2.SetVariables_#2
  &SystemState
 ]
}
emitters {
 DirectionBurst {bEnabled=false, rendererClasses [], renderers []}
 DirectionalBurst {
  emitterUpdateScript [
   &EmitterState()
   &SpawnRate(SpawnRate=6, Use Spawn Probability=false)
  ]
  particleSpawnScript [
   &InitializeParticle(
    Color=<rgba 0.9688 1.0229 1.2 0.2>
    Lifetime=0.5
    Mass Max=2
    Mass Mode=::NewEnumerator3
    Mesh Scale Mode=::NewEnumerator3
    Mesh Uniform Scale=6
    Position Offset=<0 0 -1400>
    UsePositionOffset=true
    Write Material Random=true
   )
   &DynamicMaterialParameters(Write Parameter Index 0=::NewEnumerator0)
  ]
  particleUpdateScript [
   &ParticleState()
   &ScaleColor(
    Scale Alpha=&FloatFromCurve(
     FloatCurve=DataInterface() {propertyValues {Curve {keys [
         {arriveTangent=0.002, interpMode=RCIM_Cubic, leaveTangent=0.002, time=0, value=0}
         {arriveTangent=0.002, interpMode=RCIM_Cubic, leaveTangent=0.002, time=0.1, value=1}
         {arriveTangent=0.002, interpMode=RCIM_Cubic, leaveTangent=0.002, time=1.0011, value=0.002}
        ]}}}
     Scale Curve=0.5
    )
    Scale RGB=&MakeVectorFromLinearColorRGB(Color=<rgba 1 1 1 1>)
    ScaleA=true
    ScaleRGB=true
   )
   &ScaleMeshSize(Scale Factor=&VectorFromCurve(
     VectorCurve=DataInterface() {propertyValues {
       XCurve {keys [
         {arriveTangent=3.3253, interpMode=RCIM_Cubic, leaveTangent=3.3253, tangentMode=RCTM_User, time=0, value=0}
         {arriveTangent=0.1373, interpMode=RCIM_Cubic, leaveTangent=0.1373, tangentMode=RCTM_User, time=1, value=1.5}
        ]}
       YCurve {keys [
         {arriveTangent=2.3637, interpMode=RCIM_Cubic, leaveTangent=2.3637, tangentMode=RCTM_User, time=0, value=0.1}
         {interpMode=RCIM_Cubic, tangentMode=RCTM_User, time=1, value=1}
        ]}
       ZCurve {keys [
         {arriveTangent=2.3637, interpMode=RCIM_Cubic, leaveTangent=2.3637, tangentMode=RCTM_User, time=0, value=0.1}
         {interpMode=RCIM_Cubic, tangentMode=RCTM_User, time=1, value=1}
        ]}
      }}
     Scale Curve=<1 10 10>
    ))
  ]
  rendererClasses [&NiagaraMeshRendererProperties]
  renderers [{rendererClass=&NiagaraMeshRendererProperties, rendererIndex=0}]
 }
 DistortionBlur {
  bEnabled=false
  emitterUpdateScript [
   &EmitterState()
   &SpawnBurst_Instantaneous(Use Spawn Probability=false, UseLoopCountLimit=true)
  ]
  particleSpawnScript [
   &InitializeParticle(
    Lifetime=$System.CurrentLoopDuration
    Mesh Scale Mode=::NewEnumerator3
    Sprite Size Mode=::NewEnumerator3
    Uniform Sprite Size=14000
    UsePositionOffset=true
    Write Material Random=true
   )
   &CameraOffset(Camera Offset Amount=100)
  ]
  particleUpdateScript [
   &ParticleState()
   &Timeline(Play Rate=2, Timeline End=1.0000000200408773e+20)
   /SetVariables(Particles.float:NiagaraFloat=&FloatFromCurve(FloatCurve=DataInterface() {propertyValues {Curve {keys [
         {arriveTangent=8.4865, interpMode=RCIM_Cubic, time=0, value=0}
         {arriveTangent=1, interpMode=RCIM_Cubic, leaveTangent=1, time=0.25, value=0.9847}
         {interpMode=RCIM_Cubic, leaveTangent=-3.562, time=1, value=1}
        ]}}})) {isSetParameters=true}
   &DynamicMaterialParameters(
    Index 0 Param 1=$Particles.float
    Index 0 Param 2=$Particles.float
    Index 0 Param 3=0.8
    Index 1 Param 1=0.2
    Index 1 Param 2=500
    Index 1 Param 3=0.2
    Index 2 Param 1=$Particles.Timeline.Time
    Index 2 Param 2=$Particles.Timeline.Time
    Index 2 Param 3=$Particles.Timeline.Time
    Index 2 Param 4=0.5
    Index 3 Param 1=2
    Index 3 Param 2=200
    Module.Index0Param0Write=true
    Module.Index0Param1Write=true
    Module.Index0Param2Write=true
    Module.Index0Param3Write=true
    Module.Index1Param0Write=true
    Module.Index1Param1Write=true
    Module.Index1Param2Write=true
    Module.Index1Param3Write=true
    Module.Index2Param0Write=true
    Module.Index2Param1Write=true
    Module.Index2Param2Write=true
    Module.Index2Param3Write=true
    Module.Index3Param0Write=true
    Module.Index3Param1Write=true
    Module.Index3Param2Write=true
    Module.Index3Param3Write=true
    Write Parameter Index 1=true
    Write Parameter Index 2=true
    Write Parameter Index 3=true
   )
   &ScaleColor(
    Scale Alpha=&Multiply_Float(A=1, B=&FloatFromCurve(FloatCurve=DataInterface() {propertyValues {Curve {keys [<at 0 1>,<at 1 0>]}}}))
    Scale RGB=&Multiply_VectorByFloat(
     Vector=&VectorFromFloat(Value=&Waveform("Waveform [1]"=::NewEnumerator12, "[1] Frequency"=1, "[1] Amplitude Min/Max"=<1 1>))
     Float=&FloatFromCurve(
      FloatCurve=DataInterface() {propertyValues {Curve {keys [<at 0 1>, <at 0.2 1>, {arriveTangent=0.5082, time=1, value=1}]}}}
      CurveIndex=$Particles.float
     )
    )
    ScaleA=true
    ScaleRGB=true
   )
   /SetVariables(Particles.SpeedNormalizedB:NiagaraFloat=&FloatFromCurve(
     FloatCurve=DataInterface() {propertyValues {Curve {keys [
         <at 0 0 RCIM_Cubic>
         {arriveTangent=0.7355, interpMode=RCIM_Cubic, leaveTangent=0.7355, time=0.15, value=1}
         {arriveTangent=-1.1765, interpMode=RCIM_Cubic, leaveTangent=-1.1765, time=0.7892, value=0.5804}
         <at 1 0 RCIM_Cubic>
        ]}}}
     CurveIndex=$System.NormalizedLoopAge
     Scale Curve=0.677
    )) {isSetParameters=true}
  ]
  rendererClasses [&NiagaraSpriteRendererProperties, &NiagaraComponentRendererProperties]
  renderers [
   {rendererClass=&NiagaraSpriteRendererProperties}
   {rendererClass=&NiagaraComponentRendererProperties, rendererIndex=1}
  ]
 }
 MeshRipples {
  bEnabled=false
  emitterUpdateScript [
   &EmitterState()
   &SpawnRate(SpawnRate=12, Use Spawn Probability=false)
  ]
  particleSpawnScript [
   &InitializeParticle(
    Color Maximum=<rgba 0.885 0.9091 1 0.02>
    Color Minimum=<rgba 0.759 0.8094 1 0.01>
    Color Mode=::NewEnumerator2
    Lifetime=0.5
    Mesh Scale Mode=::NewEnumerator3
    Mesh Uniform Scale=10
    Sprite Size Mode=::NewEnumerator3
    Uniform Sprite Size=100
    UsePositionOffset=false
    Write Material Random=true
   )
   &AddVelocity(Velocity=&RandomRangeVector(Minimum=<0 0 140>, Maximum=<1 1 160>))
   &InitialMeshOrientation(
    Mesh Orientation Mode=::NewEnumerator7
    Rotation=&RandomRangeVector(Minimum=<0 0 0>, Maximum=<0 1 0>)
    Rotation Coordinate Space=::NewEnumerator2
    Use Rotation Vector=true
   ) {enabled=false}
  ]
  particleUpdateScript [
   &ParticleState()
   &ScaleMeshSize(Scale Factor=&VectorFromFloat(Value=&FloatFromCurve(
      FloatCurve=DataInterface() {propertyValues {Curve {keys [{interpMode=RCIM_Cubic, tangentMode=RCTM_User, time=0, value=0}, {interpMode=RCIM_Cubic, tangentMode=RCTM_User, time=1, value=1}]}}}
      Scale Curve=2
     )))
   &ScaleColor(
    Scale Alpha=&FloatFromCurve(FloatCurve=DataInterface() {propertyValues {Curve {keys [
         {arriveTangent=9.6668, interpMode=RCIM_Cubic, leaveTangent=9.6668, tangentMode=RCTM_User, time=0, value=0}
         <at 0.1 1 RCIM_Cubic>
         {interpMode=RCIM_Cubic, tangentMode=RCTM_User, time=1, value=0}
        ]}}})
    Scale RGB=&MakeVectorFromLinearColorRGB(Color=$System.ToD)
    ScaleA=true
    ScaleRGB=true
   )
   &Drag(Drag=7, Use Linear Drag=true, Use Rotational Drag=true)
   &SolveForcesAndVelocity(Clamp Velocity=false, Limit Acceleration=false)
   &DynamicMaterialParameters(
    Index 0 Param 1=&FloatFromCurve(FloatCurve=DataInterface() {propertyValues {Curve {keys [<at 0.1 5 RCIM_Cubic>,<at 0.95 0 RCIM_Cubic>]}}})
    Index 0 Param 2=&FloatFromCurve(FloatCurve=DataInterface() {propertyValues {Curve {keys [<at 0 5 RCIM_Cubic>,<at 0.4 36 RCIM_Cubic>]}}})
    Index 0 Param 3=1
    Index 0 Param 4=&FloatFromCurve(FloatCurve=DataInterface() {propertyValues {Curve {keys [<at 0 1 RCIM_Cubic>,<at 1 0 RCIM_Cubic>]}}})
    Module.Param0WriteEnabled=true
    Module.Param1WriteEnabled=true
    Module.Param2WriteEnabled=true
    Module.Param3WriteEnabled=true
   )
  ]
  rendererClasses [&NiagaraMeshRendererProperties]
  renderers [{rendererClass=&NiagaraMeshRendererProperties, rendererIndex=0}]
 }
 MeshRipples1 {bEnabled=false, rendererClasses [], renderers []}
 NS_SoftAirflow {
  bEnabled=false
  emitterUpdateScript [
   &EmitterState()
   &SpawnBurst_Instantaneous(Spawn Count=3, Use Spawn Probability=false, UseLoopCountLimit=false)
   &SpawnRate(SpawnRate=20, Use Spawn Probability=false)
  ]
  particleSpawnScript [
   &InitializeParticle(
    Lifetime Mode=::NewEnumerator1
    Mesh Scale Mode=::NewEnumerator1
    Mesh Uniform Scale Max=40
    Mesh Uniform Scale Min=30
    Position=$System.GroundPosition
    Position Mode=::NewEnumerator1
    Position Offset=&AddVectorToPosition(Position=<0 0 0>, Vector=<0 0 0>)
    UsePositionOffset=false
    Write Material Random=true
   )
   &InitialMeshOrientation(
    Mesh Orientation Mode=::NewEnumerator7
    Rotation=&RandomRangeVector(Minimum=<0 0 0>, Maximum=<0 0 1>)
    Use Rotation Vector=true
   )
  ]
  particleUpdateScript [
   &ParticleState()
   &DynamicMaterialParameters(
    Index 0 Param 1=&FloatFromCurve(FloatCurve=DataInterface() {propertyValues {Curve {keys [{interpMode=RCIM_Cubic, tangentMode=RCTM_User, time=0.5, value=0}, {interpMode=RCIM_Cubic, tangentMode=RCTM_User, time=1, value=1}]}}})
    Index 0 Param 2=1
    Index 0 Param 3=1
    Module.Param0WriteEnabled=true
    Module.Param1WriteEnabled=true
    Module.Param2WriteEnabled=true
    Module.Param3WriteEnabled=true
   )
   &ScaleMeshSize(Scale Factor=&Multiply_VectorByFloat(
     Vector=<1 1 1.5>
     Float=&FloatFromCurve(
      FloatCurve=DataInterface() {propertyValues {Curve {keys [{interpMode=RCIM_Cubic, tangentMode=RCTM_User, time=0, value=0}, {interpMode=RCIM_Cubic, tangentMode=RCTM_User, time=1, value=1}]}}}
      Scale Curve=2
     )
    ))
   &UpdateMeshOrientation(Rotation Rate=0.5, Rotation Vector=<0 0 1>)
   &ScaleColor(
    Scale Alpha=&FloatFromCurve(
     FloatCurve=DataInterface() {propertyValues {Curve {keys [
         {arriveTangent=8.2356, interpMode=RCIM_Cubic, leaveTangent=8.2356, tangentMode=RCTM_User, time=0, value=0}
         {arriveTangent=0.1969, interpMode=RCIM_Cubic, leaveTangent=0.1969, tangentMode=RCTM_User, time=0.22, value=1}
         {arriveTangent=-1.2821, interpMode=RCIM_Cubic, leaveTangent=-1.2821, time=0.5, value=0.6}
         {arriveTangent=-0.7636, interpMode=RCIM_Cubic, leaveTangent=-0.7636, tangentMode=RCTM_User, time=1, value=0}
        ]}}}
     Scale Curve=0.1
    )
    Scale RGB=&VectorFromFloat(Value=1)
    ScaleA=true
    ScaleRGB=true
   )
  ]
  rendererClasses [&NiagaraMeshRendererProperties]
  renderers [{rendererClass=&NiagaraMeshRendererProperties, rendererIndex=0}]
 }
 PP {
  bEnabled=false
  emitterUpdateScript [
   &EmitterState()
   &SpawnBurst_Instantaneous(Use Spawn Probability=false, UseLoopCountLimit=true)
   &SpawnBurst_Instantaneous(Spawn Time=3.1167, Use Spawn Probability=false, UseLoopCountLimit=false) {module=SpawnBurst_Instantaneous001}
  ]
  particleSpawnScript [&InitializeParticle(
    Color Mode=::NewEnumerator0
    Lifetime=$System.CurrentLoopDuration
    Sprite Size Mode=::NewEnumerator3
    Uniform Sprite Size=1
    UsePositionOffset=false
    Write Material Random=true
   )]
  particleUpdateScript [
   /SetVariables(Particles.SpeedNormalizedB:NiagaraFloat=&FloatFromCurve(
     FloatCurve=DataInterface() {propertyValues {Curve {keys [
         <at 0 0 RCIM_Cubic>
         {arriveTangent=0.7355, interpMode=RCIM_Cubic, leaveTangent=0.7355, time=0.15, value=1}
         {arriveTangent=-1.1765, interpMode=RCIM_Cubic, leaveTangent=-1.1765, time=0.7892, value=0.5804}
         <at 1 0 RCIM_Cubic>
        ]}}}
     CurveIndex=$System.NormalizedLoopAge
     Scale Curve=&Multiply_Float(
      A=0.7
      B=&FN_LOD_Distance_Remap(Start Distance=12000, End Distance=14000, Minimum Output=1, Maximum Output=0, Clamp Output=true)
     )
    )) {isSetParameters=true}
   &ParticleState()
  ]
  rendererClasses [&NiagaraComponentRendererProperties]
  renderers [{rendererClass=&NiagaraComponentRendererProperties, rendererIndex=0}]
 }
 Smoke_Sphere {
  emitterUpdateScript [
   &EmitterState()
   &SpawnRate(SpawnRate=90, Use Spawn Probability=false)
   &SpawnBurst_Instantaneous(Spawn Count=30, Spawn Time=0.08, Use Spawn Probability=false) {enabled=false}
  ]
  particleSpawnScript [
   &PartitionParticles(
    Particles Per Partition=16
    Partition Count Mode:ENiagara_InfiniteLoopDuration=::NewEnumerator1
    Partition Mode:ENiagara_PartitionMode=::NewEnumerator1
   )
   &InitializeParticle(
    Color Maximum=<rgba 0.2 0.1821 0.1447 1>
    Color Minimum=<rgba 0.4 0.3751 0.3553 1>
    Color Mode=::NewEnumerator2
    Lifetime=&RandomRangeFloat(Minimum=0.4, Maximum=1)
    Position=$System.GroundPosition
    Position Mode=::NewEnumerator1
    Position Offset=&MakeVector(X=0, Y=0, Z=&Multiply_Float(A=1, B=1))
    Sprite Rotation Mode=::NewEnumerator2
    Sprite Size Mode=::NewEnumerator1
    Uniform Sprite Size Max=5000
    Uniform Sprite Size Min=3000
    UsePositionOffset=false
    Write Material Random=true
   )
   /SetVariables(Particles.InitSpriteRotRate=&UniformRangedFloat(Minimum=-3, Maximum=3)) {isSetParameters=true}
   &ShapeLocation(
    MaskSpawnGroup=false
    Non Uniform Scale=<1 1 0>
    Offset Mode=::NewEnumerator0
    Override Seed=false
    Ring Radius=800
    Shape Primitive=::NewEnumerator4
    UseOwnerScale=true
   )
   &AddVelocity(
    Module.Use Falloff Distance=false
    Override Seed=false
    Velocity Mode=::NewEnumerator1
    Velocity Origin=$System.GroundPosition
    Velocity Speed=&RandomRangeFloat(Minimum=5000, Maximum=15000)
   )
  ]
  particleUpdateScript [
   &ParticleState()
   &SpriteRotationRate(Rotation Rate=&Multiply_Float(
     A=$Particles.InitSpriteRotRate
     B=&FloatFromCurve(
      FloatCurve=DataInterface() {propertyValues {Curve {keys [
          {arriveTangent=-4.0107, interpMode=RCIM_Cubic, leaveTangent=-4.0107, tangentMode=RCTM_User, time=-0.0046, value=1.0113}
          {arriveTangent=-0.2496, interpMode=RCIM_Cubic, leaveTangent=-0.2496, tangentMode=RCTM_User, time=0.5, value=0.16}
         ]}}}
      CurveIndex=$Emitter.Age
      Scale Curve=50
     )
    ))
   &Drag(Use Linear Drag=true, Use Rotational Drag=false)
   &GravityForce(Gravity=&RandomRangeVector(Minimum=<0 0 0>, Maximum=<1 1 3000>))
   &SolveForcesAndVelocity(Clamp Velocity=false, Limit Acceleration=false)
   &ScaleSpriteSize(Scale Factor=&Vector2DFromFloat(Value=&FloatFromCurve(FloatCurve=DataInterface() {propertyValues {Curve {keys [
          {arriveTangent=6.2202, interpMode=RCIM_Cubic, leaveTangent=6.2202, tangentMode=RCTM_User, time=0, value=0.15}
          {arriveTangent=1.0025, interpMode=RCIM_Cubic, leaveTangent=1.0025, tangentMode=RCTM_User, time=0.2, value=0.8}
          {arriveTangent=0.0143, interpMode=RCIM_Cubic, leaveTangent=0.0143, tangentMode=RCTM_User, time=1.0463, value=1}
         ]}}})))
   &DynamicMaterialParameters(
    Index 0 Param 1=&Add_Float(
     A=$Particles.RandomFloat
     B=&FloatFromCurve(
      FloatCurve=DataInterface() {propertyValues {Curve {keys [
          {arriveTangent=0.3307, interpMode=RCIM_Cubic, leaveTangent=0.3307, tangentMode=RCTM_User, time=0, value=1}
          {arriveTangent=0.0529, interpMode=RCIM_Cubic, leaveTangent=0.2021, tangentMode=RCTM_User, time=1, value=1.1}
         ]}}}
      CurveIndex=&Add_Float(A=&Multiply_Float(A=$Particles.Initial.MaterialRandom, B=0.5), B=$Particles.NormalizedAge)
     )
    )
    Index 0 Param 2=&FloatFromCurve(FloatCurve=DataInterface() {propertyValues {Curve {keys [
         {arriveTangent=0.0453, interpMode=RCIM_Cubic, tangentMode=RCTM_Break, time=0, value=-1}
         {arriveTangent=0.0604, interpMode=RCIM_Cubic, leaveTangent=0.0604, tangentMode=RCTM_User, time=1, value=-1}
        ]}}})
    Index 0 Param 3=&FloatFromCurve(
     FloatCurve=DataInterface() {propertyValues {Curve {keys [{arriveTangent=0.2347, interpMode=RCIM_Cubic, leaveTangent=0.2723, tangentMode=RCTM_Break, time=0.15, value=0}]}}}
     Scale Curve=0
    )
    Index 0 Param 4=&FloatFromCurve(FloatCurve=DataInterface() {propertyValues {Curve {keys [<at 0.5 1>,<at 1 0>]}}})
    Index 1 Param 1=&FloatFromCurve(
     FloatCurve=DataInterface() {propertyValues {Curve {keys [
         {interpMode=RCIM_Cubic, leaveTangent=4.7088, tangentMode=RCTM_Break, time=0, value=0}
         {arriveTangent=4.2257, interpMode=RCIM_Cubic, leaveTangent=4.2257, tangentMode=RCTM_User, time=0.25, value=1}
        ]}}}
     Scale Curve=&RandomRangeFloat(Minimum=1.5, Maximum=1.5)
    )
    Index 1 Param 2=&FloatFromCurve(
     FloatCurve=DataInterface() {propertyValues {Curve {keys [<at 0 0.2>, {leaveTangent=0.9066, time=1, value=2}]}}}
     Scale Curve=&RandomRangeFloat(Minimum=0.5, Maximum=0.6)
    )
    Index 1 Param 3=&FloatFromCurve(FloatCurve=DataInterface() {propertyValues {Curve {keys [
         {interpMode=RCIM_Cubic, leaveTangent=6.3917, tangentMode=RCTM_Break, time=0, value=0.2}
         {arriveTangent=1.7549, interpMode=RCIM_Cubic, leaveTangent=10.6767, tangentMode=RCTM_User, time=0.25, value=1}
        ]}}})
    Index 1 Param 4=&FloatFromCurve(
     FloatCurve=DataInterface() {propertyValues {Curve {keys [{arriveTangent=-0.136, interpMode=RCIM_Cubic, leaveTangent=-0.136, tangentMode=RCTM_User, time=0, value=1}, <at 1 0.95>]}}}
     Scale Curve=&RandomRangeFloat(Minimum=1.1, Maximum=1.4)
    )
    Index 2 Param 1=&FloatFromCurve(FloatCurve=DataInterface() {propertyValues {Curve {keys [
         {interpMode=RCIM_Cubic, leaveTangent=4.2485, tangentMode=RCTM_Break, time=0, value=0}
         {arriveTangent=1.0548, interpMode=RCIM_Cubic, leaveTangent=1.0548, tangentMode=RCTM_User, time=0.17, value=0.5715}
         <at 1 1>
        ]}}})
    Index 2 Param 2=&FloatFromCurve(
     FloatCurve=DataInterface() {propertyValues {Curve {keys [
         {arriveTangent=1.9056, interpMode=RCIM_Cubic, leaveTangent=1.9056, tangentMode=RCTM_User, time=0, value=0.2}
         {arriveTangent=0.0259, interpMode=RCIM_Cubic, leaveTangent=0.0259, tangentMode=RCTM_User, time=1, value=1.4}
        ]}}}
     Scale Curve=&RandomRangeFloat(Minimum=0.2, Maximum=0.7)
    )
    Index 2 Param 3=&FloatFromCurve(FloatCurve=DataInterface() {propertyValues {Curve {keys [
         {interpMode=RCIM_Cubic, leaveTangent=2.3064, tangentMode=RCTM_Break, time=0, value=1}
         {interpMode=RCIM_Cubic, leaveTangent=11.1233, tangentMode=RCTM_Break, time=1, value=2}
        ]}}})
    Index 2 Param 4=&FloatFromCurve(
     FloatCurve=DataInterface() {propertyValues {Curve {keys [
         {arriveTangent=-0.6867, interpMode=RCIM_Cubic, leaveTangent=-0.6867, tangentMode=RCTM_User, time=0, value=1}
         {interpMode=RCIM_Cubic, leaveTangent=-1.2956, tangentMode=RCTM_Break, time=1, value=0.75}
        ]}}}
     Scale Curve=2.5
    )
    Module.Param0WriteEnabled=true
    Module.Param0WriteEnabled1=true
    Module.Param0WriteEnabled2=true
    Module.Param1WriteEnabled=false
    Module.Param1WriteEnabled1=true
    Module.Param1WriteEnabled2=true
    Module.Param2WriteEnabled=true
    Module.Param2WriteEnabled1=true
    Module.Param2WriteEnabled2=true
    Module.Param3WriteEnabled=true
    Module.Param3WriteEnabled1=true
    Module.Param3WriteEnabled2=true
    Write Parameter Index 1=true
    Write Parameter Index 2=true
   )
   &CameraOffset(Camera Offset Amount=35)
   &ScaleColor(
    Scale Alpha=&FloatFromCurve(
     FloatCurve=DataInterface() {propertyValues {Curve {keys [
         <at 0 0 RCIM_Cubic>
         {arriveTangent=1.1429, interpMode=RCIM_Cubic, leaveTangent=1.1429, time=0.1, value=1}
         {arriveTangent=-1.1111, interpMode=RCIM_Cubic, leaveTangent=-1.1111, time=0.7, value=0.8}
         <at 1 0 RCIM_Cubic>
        ]}}}
     Scale Curve=0.5
    )
    ScaleA=true
    ScaleRGB=true
   ) {module=ScaleColor001}
  ]
  rendererClasses [&NiagaraSpriteRendererProperties]
  renderers [
   {rendererClass=&NiagaraSpriteRendererProperties}
   {rendererClass=&NiagaraSpriteRendererProperties, rendererIndex=1}
  ]
 }
 SoftAirflow {rendererClasses [], renderers []}
}
summary {
 systemName=NS_PC_WealthLamb_Exit
 emitters [
  {bEnabled=false, emitterName=MeshRipples, rendererClasses [&NiagaraMeshRendererProperties]}
  {emitterName=DirectionalBurst, rendererClasses [&NiagaraMeshRendererProperties]}
  {
   bEnabled=false
   emitterName=DistortionBlur
   rendererClasses [&NiagaraSpriteRendererProperties, &NiagaraComponentRendererProperties]
  }
  {bEnabled=false, emitterName=PP, rendererClasses [&NiagaraComponentRendererProperties]}
  {emitterName=Smoke_Sphere, rendererClasses [&NiagaraSpriteRendererProperties]}
  {bEnabled=false, emitterName=NS_SoftAirflow, rendererClasses [&NiagaraMeshRendererProperties]}
  {emitterName=SoftAirflow, rendererClasses []}
  {bEnabled=false, emitterName=MeshRipples1, rendererClasses []}
  {bEnabled=false, emitterName=DirectionBurst, rendererClasses []}
 ]
 userVariables [User.GroundHeight {type=Vector3f, description="", defaultValue {struct=Vector3f, value=<0 0 1300>}}]
}
