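# Niagara sky flare effect, named for the WealthLamb exit, that shows a huge glowing mesh for a few seconds.
# Its single GiantGlow emitter burst-spawns an enormously scaled mesh particle with an overbright white color and a lifetime matching the emitter's loop duration.
# A scaled alpha curve fades the glow in quickly and out over the rest of its life, while a dynamic material parameter feeds the mesh material.
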NiagaraSystem /PistolCanary/Assets/FX/WealthLamb/Exit/NS_PS_WealthLamb_Exit_SkyFlare
dependencies {usedDataInterfaces [&NiagaraDataInterfaceCurve], usedDynamicInputs [], usedModules [&SystemState]}
emitters {GiantGlow {
  emitterUpdateScript [
   &EmitterState(Inactive Response=::NewEnumerator1, Life Cycle Mode=::NewEnumerator1, Loop Behavior=::NewEnumerator1, Loop Duration=3, UseLoopDelay=false)
   &SpawnBurst_Instantaneous(Use Spawn Probability=false, UseLoopCountLimit=false)
  ]
  particleSpawnScript [
   &InitializeParticle(
    Color=<rgba 10 10 10 1>
    Lifetime=3
    Mass Mode=::NewEnumerator1
    Mesh Scale Mode=::NewEnumerator3
    Mesh Uniform Scale=10000
    UsePositionOffset=false
    Write Material Random=true
   ) {module=InitializeParticle001}
   &DynamicMaterialParameters(
    Index 0 Param 1=1
    Index 0 Param 2=1
    Index 0 Param 3=1
    Index 0 Param 4=1
    Module.Index0Param0Write=true
    Module.Index0Param1Write=false
    Module.Index0Param2Write=false
    Module.Index0Param3Write=false
   )
  ]
  particleUpdateScript [
   &ParticleState()
   &ScaleColor(
    Scale Alpha=&FloatFromCurve(
     FloatCurve=DataInterface() {propertyValues {Curve {keys [{interpMode=RCIM_Cubic, tangentMode=RCTM_User, time=0, value=0}, {interpMode=RCIM_Cubic, tangentMode=RCTM_User, time=0.2, value=1}, <at 1 0>]}}}
     Scale Curve=0.6
    )
    ScaleA=true
    ScaleRGB=true
   ) {module=ScaleColor001}
  ]
  rendererClasses [&NiagaraMeshRendererProperties]
  renderers [{rendererClass=&NiagaraMeshRendererProperties, rendererIndex=0}]
 }}
summary {
 systemName=NS_PS_WealthLamb_Exit_SkyFlare
 emitters [{emitterName=GiantGlow, rendererClasses [&NiagaraMeshRendererProperties]}]
 userVariables []
}
